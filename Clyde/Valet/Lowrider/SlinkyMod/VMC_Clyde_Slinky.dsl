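; Lowrider "Slinky" vehicle mod for the Valet car that bounces the vehicle while its radio plays.
; First player entry binds radio start/stop events, which run or clear a looping bounce timer.
; Each grounded, permitted bounce plays hydraulic audio, applies an upward impulse near the front differential, then box-traces to launch on-vehicle-tagged pawns with a gameplay effect.
; Handbrake release also stops driver camera shake, working around lingering shake.
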(event EventBeginPlay
  (|CallParentFunction) ; This node is disabled and will not be called.\nDrag off pins to build functionality.
)

(event EventEndPlay
  (params (EndPlayReason :type EEndPlayReason%20Enum :default "Destroyed"))
  (bind _owner_vehicle (Variables|Default|GetOwner%20Vehicle))
  (bind _returnvalue (Actor|GetComponentbyClass _owner_vehicle "/SrirachaRanchValet/Components/SR_Valet.SR_Valet_C"))
  ;[0.4659 0.5156 0] End Play
  (|CallParentFunction EndPlayReason)
  (Utilities|FlowControl|Sequence
    (:then_0
      (Utilities|IsValid _owner_vehicle
        (:Is%20Valid
          (Utilities|IsValid _returnvalue
            (:Is%20Valid
              (Radio|UnbindEventfromOnSourcePlayed _returnvalue (EventDispatchers|CreateEvent :!event Custom|StartRadio))
              (Radio|UnbindEventfromOnSourceStopped _returnvalue (EventDispatchers|CreateEvent :!event Custom|StopRadio)))))))
    (:then_1
      (Utilities|Time|ClearandInvalidateTimerbyHandle (Variables|Default|GetRadio%20Timer)))))

(event EventOnPawnEnterVehicle
  (params (Vehicle :type Fort%20Vehicle%20Interface%20Interface%20%28by%20ref%29) (PlayerPawn :type Fort%20Pawn%20Object%20Reference) (SeatIndex :type Integer))
  (bind _owner_vehicle (Variables|Default|GetOwner%20Vehicle))
  (bind _returnvalue (Actor|GetComponentbyClass _owner_vehicle "/SrirachaRanchValet/Components/SR_Valet.SR_Valet_C"))
  (|CallParentFunction Vehicle PlayerPawn SeatIndex)
  (Variables|Default|SetPlayerInVehicle true)
  (Variables|Default|GetOwner%20Vehicle
    (:then
      ;[0 0.15 0.005592] Create a radio for the vehicle to use the first time a player enters the vehicle
      (Utilities|FlowControl|DoOnce false
        (:Completed
          (Radio|BindEventtoOnSourcePlayed _returnvalue (EventDispatchers|CreateEvent :!event Custom|StartRadio))
          (Radio|BindEventtoOnSourceStopped _returnvalue (EventDispatchers|CreateEvent :!event Custom|StopRadio)))))
    (:else)))

(event Custom|StartRadio
  (params (Component :type Streaming%20Radio%20Player%20Component%20Object%20Reference) (AffectedSource :type Athena%20Radio%20Station%20Structure))
  ;[0.4606 0 0.5625] Start/Stop Radio
  (Variables|Default|SetRadioIsPlaying true)
  (bind _returnvalue (Utilities|Time|SetTimerbyEvent (EventDispatchers|CreateEvent :!event Custom|While%20Radio%20is%20Playing) (Variables|Default|GetBounce%20Timing) true true))
  (Variables|Default|SetRadio%20Timer _returnvalue))

(event Custom|StopRadio
  (params (Component :type Streaming%20Radio%20Player%20Component%20Object%20Reference) (AffectedSource :type Athena%20Radio%20Station%20Structure))
  ;[0.4606 0 0.5625] Start/Stop Radio
  (Variables|Default|SetRadioIsPlaying false)
  (Utilities|Time|ClearandInvalidateTimerbyHandle (Variables|Default|GetRadio%20Timer)))

(event Custom|While%20Radio%20is%20Playing
  ;[0 0.09796 0.15] Looping Trigger addon to handle bouncing to the radio
  (Variables|Default|GetOwner%20Vehicle
    (:then
      (if (and (Variables|Default|GetPlayerInVehicle) (not (Vehicle|VehicleIsinAir (Variables|Default|GetOwner%20Vehicle))))
        (CallFunction|BouncetoMusic)
        (else
          (Development|PrintString "Looping Trigger Failed"))))
    (:else)))

(event Custom|BounceToMusic
  (bind _returnvalue_location (Math|Transform|BreakTransform (Transformation|GetBoneTransform (Class|FortVehicleInterface|GetSkeletalMeshComponent (Variables|Default|GetOwner%20Vehicle)) (Variables|Default|GetFront%20Differential%20Bone) "RTS_World")))
  (bind _owner_vehicle (Variables|Default|GetOwner%20Vehicle))
  ;[0.9657 1 0] Bounce Vehicle to radio music
  (if (Utilities|IsValid _owner_vehicle :!class /Script/Engine.KismetSystemLibrary)
    (if (Math|Boolean|ANDBoolean (Variables|Default|GetRadioIsPlaying) (CallFunction|CanModPerformActions) (not (Vehicle|PlayerIsTryingtoReverse _owner_vehicle)) (not (Variables|Default|GetHandBrake%20Active)) (not (Variables|Default|GetBlock%20Tilting)))
      (Utilities|FlowControl|Sequence
        (:then_0
          (CallFunction|PlayBounceSoundServer)
          (CallFunction|PlayBounceSound))
        (:then_1
          (FortPhysicsSimulationLibrary|AddPhysicsImpulseatLocation _owner_vehicle (* (Math|Vector|Normalize (Transformation|GetUpVector (Class|FortVehicleInterface|GetSkeletalMeshComponent _owner_vehicle))) (Variables|Default|GetBouncing%20Force)) (+ (* (Math|Vector|Normalize (Transformation|GetUpVector (Class|FortVehicleInterface|GetSkeletalMeshComponent _owner_vehicle))) (Variables|Default|GetFront%20Body%20Check%20Distance)) _returnvalue_location) true "None" false 0 "(MaxSimSizeAllowed=Large,bRejectEmbeddedInLandscape=True)")
          (CallFunction|BouncePlayersonVehicle))) ; play the audio then apply the physics force
    )))

(event Custom|Bounce%20Players%20On%20Vehicle
  (bind _owner_vehicle (Variables|Default|GetOwner%20Vehicle))
  (bind _returnvalue (Class|FortVehicleInterface|GetSkeletalMeshComponent _owner_vehicle))
  ;[0 0.1199 0.15] Bounce the players off of the vehicle into the air
  (Networking|FlushNetDormancy _owner_vehicle)
  (bind (_outbodybox _outboxextent) (FortPhysicsSimulationLibrary|GetBodyBounds _returnvalue))
  (bind (_outboxextent_x _outboxextent_y _outboxextent_z) (Math|Vector|BreakVector _outboxextent))
  (bind _returnvalue_2 (FortPhysicsSimulationLibrary|GetBoxExtent _returnvalue))
  (bind _outhits (Collision|MultiBoxTraceByChannel (Utilities|Operators|Add (Math|Vector|MakeVector 0.0 0.0 (+ _outboxextent_z (Variables|Default|GetRadio%20Bounce%20Check%20Z%20Offset))) (Transformation|GetActorLocation _owner_vehicle)) (Utilities|Operators|Add (Math|Vector|MakeVector 0.0 0.0 (+ _outboxextent_z (Variables|Default|GetRadio%20Bounce%20Check%20Z%20Offset))) (Transformation|GetActorLocation _owner_vehicle)) _returnvalue_2 (Vehicle|GetVehicleRotation _owner_vehicle) "TraceTypeQuery9" false (Utilities|Array|MakeArray (Components|GetOwner))))
  (CallFunction|ApplyLaunchToHitPawns _outhits))

(event Custom|CreateRadio)

(event EventStopHandbrake
  (bind _owner_vehicle (Variables|Default|GetOwner%20Vehicle))
  ;[0 0.1469 0.15] Fix for the camera shake randomly continuing
  (|CallParentFunction)
  (Utilities|IsValid _owner_vehicle
    (:Is%20Valid
      (bind _asvalet_basic_car_vehicle (Utilities|Casting|CastToValet_BasicCar_Vehicle _owner_vehicle))
      (Class|ValetBasicCarVehicle|StopDriverCameraShake _asvalet_basic_car_vehicle))))

(event Custom|Play%20Bounce%20Sound%20Server
  (Networking|SwitchHasAuthority
    (:Authority
      (if (!= (Vehicle|GetDriver (Variables|Default|GetOwner%20Vehicle)) (Game|GetPlayerPawn 0))
        (CallFunction|PlayBounceSound)))))

(event Custom|Play%20Bounce%20Sound
  (bind _hydraulic_bounce_sound (Variables|Default|GetHydraulic%20Bounce%20Sound)
    (:then
      (bind _returnvalue (Audio|SpawnSoundAttached _hydraulic_bounce_sound (Variables|/Script/Engine.Actor|GetRootComponent (Variables|Default|GetOwner%20Vehicle))))
      (Audio|Parameter|SetBooleanParameter _returnvalue "IsEnemy" (Utilities|Select :Index (Team|GetTeamAffiliationforActortoLocalPlayer (Vehicle|GetDriver (Variables|Default|GetOwner%20Vehicle)) true true) :Friendly false :Neutral false :Hostile true) :!class /Script/AudioExtensions.AudioParameterControllerInterface)
      (Audio|Parameter|SetBooleanParameter _returnvalue "IsLocal" (Pawn|IsLocallyViewed (Vehicle|GetDriver (Variables|Default|GetOwner%20Vehicle))) :!class /Script/AudioExtensions.AudioParameterControllerInterface))
    (:else)))

(fn ApplyLaunchtoHitPawns
  (params (Hit%20Pawns :type Array%20of%20Hit%20Result%20Structures))
  (bind _character_to_launch (LocalVariables|GetCharacter%20To%20Launch))
  (for _array_element Hit%20Pawns
    (bind _asfort_pawn (Utilities|Casting|CastToFortPawn (.HitActor (Collision|BreakHitResult _array_element))))
    (bind _output_get (LocalVariables|SetCharacter%20To%20Launch _asfort_pawn))
    (if (GameplayTags|HasTag (GameplayTags|GetOwnedGameplayTags (Ability|GetAbilitySystemComponent _output_get) :!class /Script/GameplayTags.GameplayTagAssetInterface) (Variables|Default|GetOnVehicle%20Tag) true)
      ;[0 0.2868 0.4896] apply the launch forces to the pawn
      (Pawn|Character|LaunchCharacterJump _character_to_launch (Variables|Default|GetRadio%20Player%20Bounce%20Force) false true true true)
      (Environment|ApplyGlobalEnvironmentGameplayEffecttoActor _character_to_launch (Variables|Default|GetGE_PlayerLaunched)))))
